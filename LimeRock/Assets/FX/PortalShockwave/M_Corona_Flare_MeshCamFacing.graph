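# Additive, unlit, two-sided glow material for flickering corona flares on particles and camera-facing meshes.
# Radial sphere masks build the hot center and gradient-mapped corona, with a flare texture added and tinted.
# Multiplied sines drive corona size and opacity flicker, alongside depth fade, near-camera fade and optional time-of-day dimming.
# World position offset builds the sprite, with switches for distance scaling, oscillating rotation, camera offset and black-hole/LOD effects.
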Material /LimeRock/Assets/FX/PortalShockwave/M_Corona_Flare_MeshCamFacing
graph {
 properties {
  bUsedWithMeshDeformer=true
  bUsedWithMeshParticles=true
  bUsedWithNiagaraMeshParticles=true
  bUsedWithNiagaraSprites=true
  bUsedWithParticleSprites=true
  bUsedWithSkeletalMesh=true
  blendMode=BLEND_Additive
  pixelDepthOffsetMode=PDOM_Legacy
  shadingModel=MSM_Unlit
  twoSided=true
  usedShadingModels=Unlit
 }
 nodes [
  n1 = Constant() {r=0.5}
  n2 = SphereMask(A=TextureCoordinate(), B=@n1, Radius=@n1) {hardnessPercent=0}
  n3 = Multiply(A=Power(Base=@n2) {constExponent=25}, B=ScalarParameter() {defaultValue=2, parameterName=CenterBrightness}) {constB=2}
  n4 = Constant() {r=0.5}
  n5 = ParticleRandom()
  n6 = Multiply(
   A=Add(A=VertexColor().R, B=Time())
   B=StaticSwitchParameter(True=LinearInterpolate(Alpha=@n5) {constA=1.2, constB=0.8}, False=ScalarParameter() {defaultValue=1, parameterName=FlickerIntensity}) {parameterName=GPU Particles}
  )
  n7 = Sine(None=@n6) {period=0.075}
  n8 = Multiply(A=Multiply(A=Sine(None=@n6) {period=0.155}, B=Sine(None=@n6) {period=0.2}), B=Sine(None=@n6) {period=0.43})
  n9 = ScalarParameter() {defaultValue=0.4, desc="0 = full flicker amount", parameterName=GPUFlickerFadeMin}
  n10 = ScalarParameter() {defaultValue=1, desc="1 = no flicker", parameterName=GPUFlickerFadeMax, sortPriority=1}
  n11 = ParticleRandom()
  n12 = ScalarParameter() {defaultValue=1, parameterName=FlickerLerp}
  n13 = DynamicParameter() {paramNames [FlickerFade,CoronaSize,Param3,Size]}
  n14 = LinearInterpolate(
   A=Clamp(None=Multiply(A=@n7, B=@n8)) {maxDefault=3}
   Alpha=StaticSwitchParameter(True=LinearInterpolate(A=@n9, B=@n10, Alpha=@n11) {constA=0.5}, False=Multiply(A=@n12, B=@n13.FlickerFade)) {parameterName=GPU Particles}
  ) {constA=1}
  n15 = LinearInterpolate(Alpha=@n14) {constA=-0.4, constB=0}
  n16 = ScalarParameter() {defaultValue=1, parameterName=CoronaSize}
  n17 = StaticSwitchParameter(True=@n16, False=Multiply(A=@n13.CoronaSize, B=@n16)) {parameterName=GPU Particles}
  n18 = SphereMask(A=TextureCoordinate(), B=@n4, Radius=@n4, Hardness=LinearInterpolate(Alpha=Add(A=@n15, B=@n17) {constB=0}) {constA=0.7, constB=-0.3}) {hardnessPercent=25}
  n19 = Multiply(
   A=Desaturation(
    None=TextureSampleParameter2D(UVs=@n18) {parameterName=ColorGradient, texture=/Game/Effects/Fort_Effects/Textures/Grad/T_Halo_Corona_Gradient}.RGB
    Fraction=ScalarParameter() {parameterName=Desaturate}
   )
   B=ScalarParameter() {defaultValue=0.125, parameterName=CoronaBrightness}
  ) {constB=0.125}
  n20 = VectorParameter() {parameterName=Color, defaultValue=<rgba 1 1 1 1>}
  n21 = ParticleColor()
  n22 = StaticSwitchParameter(True=@n20.RGB, False=Multiply(A=@n21.RGB, B=@n20.RGB)) {parameterName=UsedWithSkeletalMesh}
  n23 = Multiply(A=ShadingPathSwitch(Default=Add(A=@n3, B=@n19), Forward=@n19), B=LinearInterpolate(B=@n22) {constA=1, constAlpha=0.7})
  n24 = Power(Base=@n2) {constExponent=4}
  n25 = ScalarParameter() {defaultValue=0.15, parameterName=EdgeBrightness}
  n26 = TextureSampleParameter2D() {parameterName=FlareMask, texture=/Game/Effects/Fort_Effects/Textures/Flares/T_Flare_Spikeball}
  n27 = ScalarParameter() {defaultValue=1, parameterName=CenterFlareTextureBrightness}
  n28 = Multiply(A=Add(A=Multiply(A=@n24, B=@n25) {constB=0.15}, B=Multiply(A=@n26.RGB, B=@n27)), B=@n22)
  MaterialFunctionCall(Emissive=MaterialFunctionCall(Emissive=Add(A=@n23, B=@n28)) {materialFunction=&MF_FX_EmissiveAdjustments}.Result) {materialFunction=&MF_FX_EmissiveAdjustments}
  n29 = TextureSampleParameter2D() {parameterName=OuterFlareMaskTexture, samplerType=SAMPLERTYPE_Grayscale, texture=/Game/Effects/Fort_Effects/Textures/T_EFX_CameraFlare_01}
  n30 = StaticSwitchParameter(True=@n20.A, False=Multiply(A=@n21.A, B=@n20.A)) {parameterName=UsedWithSkeletalMesh}
  n31 = DepthFade(Opacity=@n30, FadeDistance=ScalarParameter() {defaultValue=20, parameterName=DepthFade})
  n32 = Power(Base=MaterialFunctionCall(Distance=ScalarParameter() {defaultValue=70, parameterName=NearCameraShrink}) {materialFunction=&Near_Camera_Fade}.Result) {constExponent=3}
  n33 = Multiply(A=@n14, B=Multiply(A=FeatureLevelSwitch(Default=@n31, ES2=@n30, ES3_1=@n30, SM4_REMOVED=@n31, SM5=@n31, SM6=@n31), B=@n32))
  n34 = StaticSwitchParameter(True=Multiply(A=@n29.G, B=@n33), False=@n33) {parameterName=UseFlareMask}
  n35 = Multiply(A=@n34) {constB=0.15}
  StaticSwitchParameter(
   True=MaterialFunctionCall(Color_Morning=@n35, Color_Day=@n34, Color_Evening=@n35, Color_Night=@n35) {materialFunction=&Color_Time_of_Day}.Result
   False=@n34
  ) {parameterName=Use color time of day}
  n36 = Transform(None=VectorParameter() {parameterName=Rotation Axis, defaultValue=<rgba 1 0 0 1>}.RGB) {transformSourceType=TRANSFORMSOURCE_Local}
  n37 = ObjectPositionWS()
  n38 = Divide(A=Add(A=ComponentMask(None=@n37) {r=true}, B=ComponentMask(None=@n37) {g=true}), B=ComponentMask(None=@n37) {b=true})
  n39 = Add(A=Time(), B=Multiply(A=Frac(None=@n38), B=ScalarParameter() {defaultValue=0.0001, parameterName=WPO Timing Offset}) {constB=0.0001})
  n40 = ScalarParameter() {defaultValue=0.125, parameterName=Rotation Speed}
  n41 = Multiply(A=Sine(None=Multiply(A=@n39, B=@n40)), B=ScalarParameter() {defaultValue=0.125, parameterName=Rotation Max})
  n42 = Add(
   A=ActorPositionWS()
   B=Transform(None=VectorParameter() {parameterName=Pivot Offset, defaultValue=<rgba 0 0 0 1>}.RGB) {transformSourceType=TRANSFORMSOURCE_Local}
  )
  n43 = WorldPosition()
  n44 = ScalarParameter() {defaultValue=30, parameterName=SpriteScale}
  n45 = MaterialFunctionCall!(
   Center Location=WorldPosition().XYZ
   0-1 UVs
   XY scale=Multiply(A=@n13.Size, B=Multiply(A=@n44, B=@n32))
   OverridePivot
   "Normals (Optional)"
  ) {materialFunction=&Sprite}
  n46 = VectorParameter() {parameterName=WorldPostionOfBlackHole, defaultValue=<rgba -30000 25536 1952 0>}
  n47 = MaterialFunctionCall!(
   World Position
   DistanceCullFadeAnimationOffset=MaterialFunctionCall() {materialFunction=/Game/Packages/Fortress_SharedMaterials/Base_Material_Functions/PickupEffectDistanceCullOverride}.Result
   Use actor center instead of pivot point
  ) {materialFunction=/Game/Packages/Fortress_SharedMaterials/Base_Material_Functions/LODEffect}
  n48 = WorldPosition()
  n49 = Add(
   A=MaterialFunctionCall!(WorldPostionOfBlackHole=AppendVector(A=@n46.RGB, B=@n46.RGB), World Position=Add(A=@n47.Result, B=@n48.XYZ)) {materialFunction=/Game/Packages/Fortnite_UniqueMaterials_1/BlackHoleFunction}.Result
   B=@n47.Result
  )
  n50 = ObjectPositionWS()
  n51 = WorldPosition()
  n52 = Multiply(A=ScalarParameter() {defaultValue=10, parameterName=MaxScaleIncrease}, B=Multiply(A=Subtract(A=@n50, B=@n51.XYZ)) {constB=-1}) {constB=-1}
  n53 = Distance(A=@n50, B=CameraPositionWS())
  n54 = ScalarParameter() {defaultValue=2000, parameterName=DistanceOffset}
  n55 = Clamp(None=Divide(A=Subtract(A=@n53, B=@n54), B=ScalarParameter() {defaultValue=30000, parameterName=ScaleUpDistance}))
  n56 = StaticSwitchParameter(
   True=StaticSwitchParameter(True=Add(A=@n45.World position offset, B=@n49), False=@n45.World position offset) {parameterName=UsedWithSkeletalMesh}
   False=StaticSwitchParameter(True=LinearInterpolate(B=@n52, Alpha=@n55) {constB=0}, False=Constant()) {parameterName=Distance Scale}
  ) {defaultValue=true, parameterName=CameraFacingMeshSprite}
  n57 = StaticSwitchParameter(True=Add(A=RotateAboutAxis(NormalizedRotationAxis=@n36, RotationAngle=@n41, PivotPoint=@n42, Position=@n43.XYZ), B=@n56), False=@n56) {parameterName=RotateAboutAxis}
  n58 = ScalarParameter() {parameterName=CameraOffset}
  StaticSwitchParameter(
   True=Add(
    A=@n57
    B=MaterialFunctionCall(Offset Amount=@n58) {materialFunction=/Engine/Functions/Engine_MaterialFunctions02/WorldPositionOffset/CameraOffset}.Clamped Camera Offset
   )
   False=@n57
  ) {parameterName=CameraOffset}
 ]
}
