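# Additive, unlit, two-sided material for Niagara mesh particles, named as a fresnel impact shockwave for the LimeRock portal.
# It forms a soft rim band by multiplying an inner Fresnel with a squared, inverted edge Fresnel, with exponents driven by dynamic parameters.
# Particle color tints the band through MF_FX_EmissiveAdjustments, while particle alpha feeds a depth fade.
# The faded mask and a dynamic parameter drive slight index-of-refraction distortion, disabled on low quality.
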Material /LimeRock/Assets/FX/PortalShockwave/M_LimeRock_Portal_Shockwave_Impct_Fresnel
graph {
 properties {
  bUseTranslucencyVertexFog=false
  bUsedWithNiagaraMeshParticles=true
  blendMode=BLEND_Additive
  pixelDepthOffsetMode=PDOM_Legacy
  refractionMethod=RM_IndexOfRefraction
  shadingModel=MSM_Unlit
  twoSided=true
  usedShadingModels=Unlit
 }
 nodes [
  n1 = DynamicParameter() {defaultValue=<rgba 1 4 1 1>, paramNames [InnerFresnel,EdgeFresnel,Refraction,Param4]}
  n2 = OneMinus(None=Fresnel(ExponentIn=@n1.EdgeFresnel) {baseReflectFraction=0, exponent=4})
  n3 = Multiply(A=Fresnel(ExponentIn=@n1.InnerFresnel) {baseReflectFraction=0}, B=Multiply(A=@n2, B=@n2))
  n4 = ParticleColor()
  MaterialFunctionCall(Emissive=Multiply(A=@n3, B=@n4.RGB)) {materialFunction=&MF_FX_EmissiveAdjustments}
  n5 = DepthFade(Opacity=Multiply(A=@n3, B=@n4.A), FadeDistance=ScalarParameter() {defaultValue=5000, parameterName=DepthFadeDist}) {fadeDistanceDefault=5000}
  n6 = DynamicParameter() {defaultValue=<rgba 1 4 1 1>, paramNames [InnerFresnel,EdgeFresnel,Refraction,Param4]}
  QualitySwitch(Default=LinearInterpolate(Alpha=Multiply(A=@n5, B=@n6.Refraction)) {constA=1, constB=1.1}, Low=Constant() {r=1})
 ]
}
