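# Niagara ground layer of the LimeRock portal shockwave, lifting grass, debris and glowing motes into the air.
# Grass sprites and mesh-plus-sprite debris burst away from System.PortalPosition and upward, hover on drag and wind, then fall when System.Drop enables gravity.
# GPU motes spawn blended between a wide cylinder and a sphere around the player.
# A fixed world-space box culls spawns; debris and grass shrink near the player, motes near the camera.
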NiagaraSystem /LimeRock/Assets/FX/PortalShockwave/NS_LimeRock_Shockwave_Ground
dependencies {
 usedDataInterfaces [
  &NiagaraDataInterfaceSpriteRendererInfo
  &NiagaraDataInterfaceCurve
  &NiagaraDataInterfaceMeshRendererInfo
  &NiagaraDataInterfaceCamera
 ]
 usedDynamicInputs [
  &SetBoolByFloatComparison
  /Niagara/DynamicInputs/TypeConversions/MakeCustomPositionFromBool
  &VectorLength
  &MakeVectorFromLinearColorRGB
 ]
 usedModules [
  &self:SystemSpawnScript.SystemScriptSource.SystemScriptGraph.NiagaraNodeAssignment_2.SetVariables_#1
  &self:SystemSpawnScript.SystemScriptSource.SystemScriptGraph.NiagaraNodeAssignment_0.SetVariables_#2
  &SystemState
  &self:SystemSpawnScript.SystemScriptSource.SystemScriptGraph.NiagaraNodeAssignment_1.SetVariables_#3
  &self:SystemSpawnScript.SystemScriptSource.SystemScriptGraph.NiagaraNodeAssignment_3.SetVariables_#4
 ]
}
emitters {
 AmbientMotes {
  simTarget=GPUComputeSim
  emitterUpdateScript [
   &EmitterState(
    Life Cycle Mode=::NewEnumerator1
    Loop Behavior=::NewEnumerator1
    Loop Duration=&Subtract_Float(A=$System.Duration, B=4)
    UseLoopDelay=false
   )
   &SpawnBurst_Instantaneous(Spawn Count=50, Spawn Group=1, Use Spawn Probability=false, UseLoopCountLimit=false)
   /SetVariables(Emitter.CamPosition:NiagaraPosition=&ReturnCameraPosition(
     Coordinate Space:ENiagaraCoordinateSpace=::Simulation
     Camera Query:NiagaraDataInterfaceCamera=DataInterface() {propertyValues {bRequireCurrentFrameData=false}}
    )) {isSetParameters=true}
   &SpawnRate(SpawnRate=100, Use Spawn Probability=false)
  ]
  particleSpawnScript [
   &InitializeParticle(
    Color=&MaskLinearColorBySpawnGroup(
     Masked Linear Color=<rgba 0.0214 0.0203 0.026 0.7>
     Passthrough Linear Color=&RandomRangeLinearColor(Minimum=<rgba 1 0.0396 0.978 0.8>, Maximum=<rgba 1.5 0.058 0.5101 1>)
     Spawn Group Mask=1
    )
    Lifetime=&RandomRangeFloat(Minimum=3, Maximum=5)
    Sprite Size Mode=::NewEnumerator1
    Uniform Sprite Size Max=30
    Uniform Sprite Size Min=20
    UsePositionOffset=false
    Write Material Random=true
   )
   &ShapeLocation(
    Apply To Particle Position=::NewEnumerator1
    Cylinder Height=500
    Cylinder Height Midpoint=0.7
    Cylinder Radius=4000
    MaskSpawnGroup=false
    Offset Mode=::NewEnumerator0
    Override Seed=false
    Shape Primitive=::NewEnumerator1
    Surface Only=false
    UseOwnerScale=true
   )
   &ShapeLocation(
    Apply To Particle Position=::NewEnumerator1
    MaskSpawnGroup=false
    Offset=<0 0 -100>
    Offset Mode=::NewEnumerator0
    Override Seed=false
    Shape Origin=$System.PlayerPosition
    Sphere Radius=800
    Sphere Surface Distribution=0.5
    UseOwnerScale=true
   ) {module=ShapeLocation001}
   /SetVariables(Particles.Position=&Lerp_Vector(
     A=Linked() {linkedVariable {name=Particles.ShapeLocation.ShapePosition, type=NiagaraPosition}}
     B=Linked() {linkedVariable {name=Particles.ShapeLocation001.ShapePosition, type=NiagaraPosition}}
     Alpha=&RandomRangeFloat(Minimum=0, Maximum=1)
    )) {isSetParameters=true}
   &KillParticlesInVolume(Box Size:Vector3f=<1500 1200 5000>, Kill Shape=::NewEnumerator1, Volume Origin:Vector3f=<-90852.75 -94788.9141 1457.743>)
   &AddVelocity(Velocity=<0 0 1>, Velocity Speed Scale=&RandomRangeFloat(Minimum=20, Maximum=50)) {module=AddVelocity001}
  ]
  particleUpdateScript [
   &ParticleState()
   &ScaleSpriteSize(
    Uniform Curve Scale=&FloatFromCurve(
     FloatCurve=DataInterface() {propertyValues {Curve {keys [{arriveTangent=0.0004, interpMode=RCIM_Cubic, leaveTangent=0.0004, tangentMode=RCTM_User, time=500, value=0.1}, <at 3000 1>]}}}
     CurveIndex=&DistanceBetweenPositions(A=$Emitter.CamPosition, B=$Particles.Position)
     Scale Curve=&MaskFloatBySpawnGroup(Masked Float=0.6, Passthrough Float=1, Spawn Group Mask=1)
    )
    Uniform Curve Sprite Scale=DataInterface() {propertyValues {Curve {
       defaultValue=0
       keys [
        {interpMode=RCIM_Cubic, tangentMode=RCTM_User, time=0, value=0}
        {arriveTangentWeight=0.0833, interpMode=RCIM_Cubic, leaveTangentWeight=0.1742, tangentMode=RCTM_User, time=0.25, value=1}
        {
         arriveTangent=-0.2912
         arriveTangentWeight=0.2903
         interpMode=RCIM_Cubic
         leaveTangent=-0.2912
         leaveTangentWeight=0.1263
         tangentMode=RCTM_User
         time=0.71
         value=0.9
        }
        {tangentMode=RCTM_User, time=1, value=0}
       ]
      }}}
   )
   &WindForce(
    Friction Mode:ENiagaraWindFrictionMode=::NewEnumerator0
    Scale:NiagaraFloat=&MakeCustomFloatFromBool(Bool=$System.Drop, True Float=0, False Float=1)
    Surface Alignment Mode:ENiagaraWindCombingMode=::NewEnumerator0
    Wind Speed=<0 0 1>
    Wind Speed Scale=&MakeCustomFloatFromBool(Bool=$System.Drop, True Float=0, False Float=&RandomRangeFloat(Minimum=22, Maximum=30))
   )
   &Drag(Use Linear Drag=true, Use Rotational Drag=false)
   &SolveForcesAndVelocity(Clamp Velocity=false, Limit Acceleration=false)
  ]
  rendererClasses [&NiagaraSpriteRendererProperties]
  renderers [{rendererClass=&NiagaraSpriteRendererProperties}]
 }
 FloatingDebris {
  emitterUpdateScript [
   &EmitterState(Life Cycle Mode=::NewEnumerator1, Loop Behavior=::NewEnumerator1, Loop Duration=1, UseLoopDelay=false)
   &SpawnRate(SpawnRate=25, Use Spawn Probability=false)
   &SpawnBurst_Instantaneous(Spawn Count=50, Spawn Time=0.05, Use Spawn Probability=false, UseLoopCountLimit=false)
  ]
  particleSpawnScript [
   /SetVariables(Particles.SizeLerp:NiagaraFloat=&FloatFromCurve(
     FloatCurve=DataInterface() {propertyValues {Curve {
        defaultValue=1
        keys [
         <at 0 0>
         {arriveTangent=0.7499, interpMode=RCIM_Cubic, leaveTangent=0.7499, tangentMode=RCTM_User, time=0.6584, value=0.3482}
         {interpMode=RCIM_Cubic, tangentMode=RCTM_User, time=0.95, value=1}
        ]
       }}}
     CurveIndex=&RandomRangeFloat(Minimum=0, Maximum=1)
    )) {isSetParameters=true}
   &InitializeParticle(
    Color=<rgba 2 0.0895 1.9726 0.1>
    Lifetime=&Add_Float(A=$System.Duration, B=1)
    Mesh Renderer Array Visibility Mode=::NewEnumerator3
    Mesh Renderer Info=DataInterface() {propertyValues {MeshRenderer=&self:FloatingDebris_0.NiagaraMeshRendererProperties_0}}
    Mesh Scale Mode=::NewEnumerator3
    Mesh Uniform Scale=&Lerp_Float(A=0.7, B=2, Alpha=$Particles.SizeLerp)
    Sprite Size Mode=::NewEnumerator3
    Uniform Sprite Size=&Lerp_Float(A=25, B=45, Alpha=$Particles.SizeLerp)
    UsePositionOffset=false
    Write Material Random=true
   )
   &ShapeLocation(
    Cylinder Height Midpoint=0.9
    Cylinder Radius=2000
    MaskSpawnGroup=false
    Override Seed=false
    Shape Primitive=::NewEnumerator1
    Surface Only=false
    UseOwnerScale=true
   )
   &KillParticlesInVolume(Box Size:Vector3f=<1500 1200 5000>, Kill Shape=::NewEnumerator1, Volume Origin:Vector3f=<-90852.75 -94788.9141 1457.743>)
   /SetVariables(Particles.DebrisColor:LinearColor=&RandomRangeLinearColor(Minimum=<rgba 2 1.8528 1.628 1>, Maximum=<rgba 3 1.9656 2.0516 1>)) {isSetParameters=true, module=SetVariables_#5}
   &AddVelocity(
    Module.Use Falloff Distance=false
    Origin Offset Coordinate Space=::World
    Override Seed=false
    Velocity Mode=::NewEnumerator1
    Velocity Origin=$System.PortalPosition
    Velocity Speed=&RandomRangeFloat(Minimum=200, Maximum=300)
   )
   &AddVelocity(Velocity=<0 0 1>, Velocity Speed Scale=&RandomRangeFloat(Minimum=300, Maximum=350)) {module=AddVelocity001}
   &InitialMeshOrientation(Use Rotation Vector=false)
   /SetVariables() {isSetParameters=true, module=SetVariables_#6}
  ]
  particleUpdateScript [
   &ParticleState()
   &DynamicMaterialParameters(
    Index 0 Param 1=&FloatFromCurve(
     FloatCurve=DataInterface() {propertyValues {Curve {keys [
         {arriveTangent=-13.0787, interpMode=RCIM_Cubic, leaveTangent=-13.0787, tangentMode=RCTM_User, time=0, value=1}
         {arriveTangentWeight=0.1179, interpMode=RCIM_Cubic, leaveTangentWeight=0.3536, tangentMode=RCTM_User, time=0.1, value=0.1}
         {arriveTangent=-0.1404, tangentMode=RCTM_User, time=1, value=0}
        ]}}}
     Scale Curve=&RandomRangeFloat(Minimum=0.3, Maximum=0.55)
    )
    Index 3 Param 1=3
    Index 3 Param 2=1
    Module.Index0Float0Write=true
    Module.Index0Float1Write=true
    Module.Index0Float2Write=false
    Module.Index0Float3Write=false
    Module.Index3Float0Write=true
    Module.Index3Float1Write=true
    Module.Index3Float2Write=false
    Module.Index3Float3Write=false
    Write Parameter Index 3=::NewEnumerator1
   )
   /SetVariables(Particles.NearCamScale:NiagaraFloat=&FloatFromCurve(
     FloatCurve=DataInterface() {propertyValues {Curve {keys [{arriveTangent=0.0102, interpMode=RCIM_Cubic, leaveTangent=0.0102, tangentMode=RCTM_User, time=50, value=0}, <at 200 1>]}}}
     CurveIndex=&DistanceBetweenPositions(A=$Particles.Position, B=$System.PlayerPosition)
    )) {isSetParameters=true}
   &ScaleSpriteSize(Scale Sprite Size Mode=::NewEnumerator0, Uniform Scale Factor=$Particles.NearCamScale)
   &ScaleMeshSize(Scale Factor=&VectorFromFloat(Value=&FloatFromCurve(
      FloatCurve=DataInterface() {propertyValues {Curve {
         defaultValue=1
         keys [
          {interpMode=RCIM_Cubic, tangentMode=RCTM_User, time=0, value=0}
          {arriveTangentWeight=0.0267, interpMode=RCIM_Cubic, leaveTangentWeight=0.4002, tangentMode=RCTM_User, time=0.05, value=1}
         ]
        }}}
      Scale Curve=$Particles.NearCamScale
     )))
   &UpdateMeshOrientation(
    Rotation Rate=&FloatFromCurve(
     FloatCurve=DataInterface() {propertyValues {Curve {keys [
         {arriveTangent=-6.9236, interpMode=RCIM_Cubic, leaveTangent=-6.9236, tangentMode=RCTM_User, time=0, value=1}
         <at 0.2 0.15>
         <at 0.9995 0.07>
        ]}}}
     Scale Curve=1.5
    )
    Rotation Vector=&RandomVector(Vector Scale=1)
   )
   &WindForce(
    Scale:NiagaraFloat=&MakeCustomFloatFromBool(Bool=$System.Drop, True Float=0, False Float=1)
    Wind Speed=<0 0 1>
    Wind Speed Scale=&MakeCustomFloatFromBool(Bool=$System.Drop, True Float=0, False Float=&RandomRangeFloat(Minimum=8, Maximum=18))
   )
   &GravityForce(Gravity=&MakeVector(
     X=0
     Y=0
     Z=&MakeCustomFloatFromBool(Bool=$System.Drop, True Float=&RandomRangeFloat(Minimum=-1600, Maximum=-900), False Float=0)
    ))
   &Drag(Drag=2, Use Linear Drag=true, Use Rotational Drag=false)
   &SolveForcesAndVelocity(Clamp Velocity=false, Limit Acceleration=false)
   &ScaleColor(
    Scale Alpha=&FloatFromCurve(FloatCurve=DataInterface() {propertyValues {Curve {keys [
         {arriveTangent=56.5835, interpMode=RCIM_Cubic, leaveTangent=56.5835, tangentMode=RCTM_User, time=0, value=0}
         {arriveTangentWeight=0.0435, interpMode=RCIM_Cubic, leaveTangentWeight=0.3634, tangentMode=RCTM_User, time=0.05, value=1}
         {arriveTangentWeight=0.299, interpMode=RCIM_Cubic, leaveTangentWeight=0.1495, tangentMode=RCTM_User, time=0.85, value=1}
         <at 0.95 0>
        ]}}})
    ScaleA=true
    ScaleRGB=false
   )
  ]
  rendererClasses [&NiagaraMeshRendererProperties, &NiagaraSpriteRendererProperties]
  renderers [
   {rendererClass=&NiagaraMeshRendererProperties, rendererIndex=0}
   {rendererClass=&NiagaraSpriteRendererProperties, rendererIndex=1}
  ]
 }
 FloatingGrass {
  emitterUpdateScript [
   &EmitterState(Life Cycle Mode=::NewEnumerator1, Loop Behavior=::NewEnumerator1, Loop Duration=1, UseLoopDelay=false)
   &SpawnRate(SpawnRate=120, Use Spawn Probability=false)
   &SpawnBurst_Instantaneous(Spawn Count=70, Spawn Time=0.03, Use Spawn Probability=false, UseLoopCountLimit=false)
  ]
  particleSpawnScript [
   &InitializeParticle(
    Color=<rgba 4.8548 0 5 0.05>
    Lifetime=&Add_Float(A=$System.Duration, B=1)
    Mass=0.5
    Mass Mode=::NewEnumerator1
    Sprite Size Max=<20 20>
    Sprite Size Min=<12 12>
    Sprite Size Mode=::NewEnumerator2
    UsePositionOffset=false
    Write Material Random=true
   )
   &ShapeLocation(
    Cylinder Height Midpoint=0.9
    Cylinder Radius=2000
    MaskSpawnGroup=false
    Override Seed=false
    Shape Primitive=::NewEnumerator1
    Surface Only=false
    UseOwnerScale=true
   )
   &KillParticlesInVolume(Box Size:Vector3f=<1500 1200 5000>, Kill Shape=::NewEnumerator1, Volume Origin:Vector3f=<-90852.75 -94788.9141 1457.743>)
   /SetVariables(Particles.GrassColor:LinearColor=&RandomRangeLinearColor(Minimum=<rgba 0.056 0.1 0.038 2.5>, Maximum=<rgba 0.0026 0.1 0.0667 2.5>)) {isSetParameters=true, module=SetVariables_#5}
   &SubUVAnimation(
    "Sprite Renderer "=DataInterface() {propertyValues {SpriteRenderer=&self:FloatingGrass_0.NiagaraSpriteRendererProperties_1}}
    SubUV Animation Mode:ENiagara_SubUVLookupModeV2=::NewEnumerator2
    UseEndFrame=false
    UseStartFrame=false
   )
   &AddVelocity(
    Module.Use Falloff Distance=false
    Origin Offset Coordinate Space=::World
    Override Seed=false
    Velocity Mode=::NewEnumerator1
    Velocity Origin=$System.PortalPosition
    Velocity Speed=&RandomRangeFloat(Minimum=400, Maximum=600)
   )
   &AddVelocity(Velocity=<0 0 1>, Velocity Speed Scale=&RandomRangeFloat(Minimum=300, Maximum=350)) {module=AddVelocity001}
  ]
  particleUpdateScript [
   &ParticleState()
   &ScaleSpriteSize(
    Scale Sprite Size Mode=::NewEnumerator0
    Uniform Scale Factor=&FloatFromCurve(
     FloatCurve=DataInterface() {propertyValues {Curve {keys [{arriveTangent=0.0102, interpMode=RCIM_Cubic, leaveTangent=0.0102, tangentMode=RCTM_User, time=50, value=0}, <at 200 1>]}}}
     CurveIndex=&DistanceBetweenPositions(A=$Particles.Position, B=$System.PlayerPosition)
    )
   )
   /SetVariables(Particles.GlowSize=&Multiply_Vector2D_ByFloat(Float=1.2, Vector2D=$Particles.SpriteSize)) {isSetParameters=true, module=SetVariables_#6}
   &SpriteRotationRate(Rotation Rate=&FloatFromCurve(
     FloatCurve=DataInterface() {propertyValues {Curve {keys [
         {arriveTangent=-6.9236, interpMode=RCIM_Cubic, leaveTangent=-6.9236, tangentMode=RCTM_User, time=0, value=1}
         <at 0.2 0.15>
         <at 0.9995 0.07>
        ]}}}
     Scale Curve=&FloatFromCurve(
      FloatCurve=DataInterface() {propertyValues {Curve {keys [
          {interpMode=RCIM_Cubic, tangentMode=RCTM_User, time=0, value=-1}
          {arriveTangent=5.8272, interpMode=RCIM_Cubic, leaveTangent=5.8272, tangentMode=RCTM_User, time=0.5, value=0}
          {interpMode=RCIM_Cubic, tangentMode=RCTM_User, time=1, value=1}
         ]}}}
      CurveIndex=&RandomRangeFloat(Minimum=0, Maximum=1)
      Scale Curve=400
     )
    ))
   &WindForce(
    Scale=1
    Wind Speed=<0 0 1>
    Wind Speed Scale=&MakeCustomFloatFromBool(Bool=$System.Drop, True Float=0, False Float=&RandomRangeFloat(Minimum=10, Maximum=20))
   )
   &GravityForce(Gravity=&MakeVector(
     X=0
     Y=0
     Z=&MakeCustomFloatFromBool(Bool=$System.Drop, True Float=&RandomRangeFloat(Minimum=-1200, Maximum=-700), False Float=0)
    ))
   &Drag(Drag=2, Use Linear Drag=true, Use Rotational Drag=false)
   &SolveForcesAndVelocity(Clamp Velocity=false, Limit Acceleration=false)
   &ScaleColor(
    Scale Alpha=&FloatFromCurve(FloatCurve=DataInterface() {propertyValues {Curve {keys [
         {arriveTangent=33.2805, interpMode=RCIM_Cubic, leaveTangent=33.2805, tangentMode=RCTM_User, time=0, value=0}
         {arriveTangentWeight=0.0435, interpMode=RCIM_Cubic, leaveTangentWeight=0.3634, tangentMode=RCTM_User, time=0.05, value=1}
         {arriveTangentWeight=0.299, interpMode=RCIM_Cubic, leaveTangentWeight=0.1495, tangentMode=RCTM_User, time=0.85, value=1}
         <at 0.95 0>
        ]}}})
    ScaleA=true
    ScaleRGB=false
   )
  ]
  rendererClasses [&NiagaraSpriteRendererProperties]
  renderers [
   {rendererClass=&NiagaraSpriteRendererProperties}
   {rendererClass=&NiagaraSpriteRendererProperties, rendererIndex=1}
  ]
 }
}
summary {
 systemName=NS_LimeRock_Shockwave_Ground
 emitters [
  {emitterName=FloatingGrass, rendererClasses [&NiagaraSpriteRendererProperties]}
  {emitterName=FloatingDebris, rendererClasses [&NiagaraMeshRendererProperties, &NiagaraSpriteRendererProperties]}
  {emitterName=AmbientMotes, simTarget=GPUComputeSim, rendererClasses [&NiagaraSpriteRendererProperties]}
 ]
 userVariables [
  User.Intensity {type=NiagaraFloat, defaultValue=0, description=""}
  User.PointForce {type=NiagaraFloat, defaultValue=0, description=""}
  User.PointForceOrigin {type=NiagaraPosition, description="", defaultValue {}}
 ]
}
